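# Physics asset for the Baba Yaga pickaxe mesh, defining secondary jiggle motion for its broom, straw, grass, leaf, mushrooms and feet.
# Boxes cover grass, leaf and base bones, while capsules cover straw, feet, mushrooms and the broom body.
# Constraints with limited swing and twist hang each piece from a base bone, with short chains for some upper grass, both feet and one mushroom.
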PhysicsAsset /BRCosmetics/Weapons/FORT_Melee/Pickaxe_Baba_Yaga/Meshes/Baba_Yaga_Physics
bodies {
 dyn_bottom_handle_grass_01 [{shapeType=Box, extentX=4.7731, extentY=5.1536, extentZ=0.8155, center=<2.5694 0 0>}]
 dyn_bottom_handle_grass_02 [{shapeType=Box, extentX=5.2542, extentY=3.7149, extentZ=0.5, center=<2.3552 0 0>}]
 dyn_bottom_handle_grass_03 [{shapeType=Box, extentX=2.8698, extentY=3.6172, extentZ=0.75, center=<1.3381 0 0>}]
 dyn_bottom_handle_grass_base [{shapeType=Box, extentX=0.5, extentY=0.5, extentZ=0.5}]
 dyn_broom_base [{shapeType=Box, extentX=0.75, extentY=0.75, extentZ=0.75}]
 dyn_broom_master [{shapeType=Box, extentX=3, extentY=3, extentZ=3}]
 dyn_broom_parent [{shapeType=Capsule, length=7.375, radius=13.5, center=<8.8461 -0.0011 -0.0002>, rotation=<rot 90 0 0>}]
 dyn_feet_base [
  {shapeType=Capsule, length=16.625, radius=2.75, center=<7.9838 1.1571 -0.255>, rotation=<rot 81.5498 6.1741 -2.3718>}
  {shapeType=Box, extentX=2.25, extentY=2.25, extentZ=2.75}
 ]
 dyn_foot_a_01 [{shapeType=Capsule, length=2.8052, radius=1.375, center=<2.2236 0 0>, rotation=<rot 90 0 0>}]
 dyn_foot_a_02 [{shapeType=Capsule, length=6.9023, radius=5.5236, center=<8.6706 0 0>, rotation=<rot 80.8414 179.9999 179.9999>}]
 dyn_foot_b_01 [{shapeType=Capsule, length=2.9287, radius=1.5, center=<2.6352 0 0>, rotation=<rot 90 0 0>}]
 dyn_foot_b_02 [{shapeType=Capsule, length=3.75, radius=3.6127, center=<5.1049 0 0>, rotation=<rot 76.0882 -179.9999 -179.9999>}]
 dyn_grass_lower_a [{shapeType=Box, extentX=2.5658, extentY=2.0209, extentZ=1.2233, center=<1.3998 0 0>}]
 dyn_grass_lower_b [{shapeType=Box, extentX=3, extentY=2.7447, extentZ=1.0888, center=<1.6069 0 0>}]
 dyn_grass_lower_base [{shapeType=Box, extentX=1.5, extentY=1.5, extentZ=1.5}]
 dyn_grass_lower_c [{shapeType=Box, extentX=3.6254, extentY=4.0263, extentZ=1.3529, center=<2.3781 0.0001 0>}]
 dyn_grass_lower_d [{shapeType=Box, extentX=4.2302, extentY=5.4778, extentZ=1.5978, center=<2.5667 0 0>}]
 dyn_grass_lower_e [{shapeType=Box, extentX=3.2852, extentY=3.6468, extentZ=0.5, center=<1.7321 0 0.0002>}]
 dyn_grass_lower_f [{shapeType=Box, extentX=4.2565, extentY=5.0578, extentZ=1.5, center=<2.3914 0 0.0001>}]
 dyn_grass_upper_a [{shapeType=Box, extentX=3, extentY=3.2104, extentZ=1, center=<1.9788 0 0>}]
 dyn_grass_upper_a_01 [{shapeType=Box, extentX=3, extentY=2.5223, extentZ=1.065, center=<1.7793 0 0>}]
 dyn_grass_upper_a_02 [{shapeType=Box, extentX=3, extentY=2.0362, extentZ=0.75, center=<1.6118 0 0>}]
 dyn_grass_upper_b [{shapeType=Box, extentX=3, extentY=2.26, extentZ=1, center=<1.6023 0 0>}]
 dyn_grass_upper_b_01 [{shapeType=Box, extentX=3, extentY=2.2619, extentZ=0.5, center=<1.5744 0 0>}]
 dyn_grass_upper_b_02 [{shapeType=Box, extentX=3, extentY=2.2664, extentZ=1.0708, center=<1.7773 0 0>}]
 dyn_grass_upper_base [{shapeType=Box, extentX=1.5, extentY=1.5, extentZ=1.5}]
 dyn_grass_upper_c [{shapeType=Box, extentX=3, extentY=2, extentZ=1.0504, center=<1.7962 0 0>}]
 dyn_grass_upper_d [{shapeType=Box, extentX=3.4374, extentY=4.9567, extentZ=1.3955, center=<2.4942 0 0>}]
 dyn_grass_upper_e [{shapeType=Box, extentX=3, extentY=3.0805, extentZ=1.3078, center=<2.109 0.0001 0>}]
 dyn_grass_upper_f [{shapeType=Box, extentX=4.5, extentY=2.6642, extentZ=1.1375, center=<2.1011 0 0>}]
 dyn_grass_upper_g [{shapeType=Box, extentX=4.0809, extentY=4.4316, extentZ=1.4173, center=<3.0263 0 0>}]
 dyn_leaf [{shapeType=Box, extentX=4.0247, extentY=3.5835, extentZ=0.8803, center=<1.9137 0 0>}]
 dyn_mushroom_a [{shapeType=Capsule, length=3.5152, radius=2.2274, center=<3.7883 0.0306 1.0723>, rotation=<rot -51.5106 34.206 -41.7995>}]
 dyn_mushroom_b [{shapeType=Capsule, length=2, radius=1.75, center=<2.19 0 0>, rotation=<rot 90 0 0>}]
 dyn_mushroom_b_01 [{shapeType=Capsule, length=2.2186, radius=1.615, center=<2.2797 0 0>, rotation=<rot 90 0 0>}]
 dyn_mushroom_c [{shapeType=Capsule, length=2.38, radius=1.5981, center=<2.7572 -0.0001 0>, rotation=<rot 90 0 0>}]
 dyn_staw_01 [{shapeType=Capsule, length=5.1403, radius=1.3025, center=<3.051 0.0001 0>, rotation=<rot 90 0 0>}]
 dyn_straw_02 [{shapeType=Capsule, length=6.7111, radius=1.8735, center=<4.7659 0 0>, rotation=<rot 90 0 0>}]
 dyn_straw_03 [{shapeType=Capsule, length=3.4617, radius=1.3496, center=<2.7907 0 0>, rotation=<rot 90 0 0>}]
 dyn_straw_04 [{shapeType=Capsule, length=5.4784, radius=1.3974, center=<3.7327 0 0>, rotation=<rot 90 0 0>}]
 dyn_straw_05 [{shapeType=Capsule, length=4.3748, radius=1.1102, center=<2.6918 0.0001 0>, rotation=<rot 90 0 0>}]
 dyn_top_handle_grass_01 [{shapeType=Box, extentX=4.3867, extentY=5.0668, extentZ=0.7552, center=<2.3316 0 0>}]
 dyn_top_handle_grass_02 [{shapeType=Box, extentX=5.1497, extentY=3.8282, extentZ=1.1285, center=<2.4448 0 0>}]
 dyn_top_handle_grass_03 [{shapeType=Box, extentX=4.6167, extentY=4.2301, extentZ=1.0905, center=<1.8095 0 0>}]
 dyn_top_handle_grass_base [{shapeType=Box, extentX=0.5, extentY=0.5, extentZ=0.5}]
}
constraints [
 {bone1Name=dyn_broom_parent, bone2Name=dyn_broom_base, swing1LimitDegrees=10, swing2LimitDegrees=10, twistLimitDegrees=5}
 {bone1Name=dyn_staw_01, bone2Name=dyn_broom_master, swing1LimitDegrees=15, swing2LimitDegrees=10, twistLimitDegrees=3}
 {bone1Name=dyn_straw_02, bone2Name=dyn_broom_master, swing1LimitDegrees=15, swing2LimitDegrees=10, twistLimitDegrees=3}
 {bone1Name=dyn_straw_03, bone2Name=dyn_broom_master, swing1LimitDegrees=20, swing2LimitDegrees=20, twistLimitDegrees=3}
 {bone1Name=dyn_straw_04, bone2Name=dyn_broom_master, swing1LimitDegrees=10, swing2LimitDegrees=10, twistLimitDegrees=2}
 {bone1Name=dyn_straw_05, bone2Name=dyn_broom_master, swing1LimitDegrees=20, swing2LimitDegrees=10, twistLimitDegrees=3}
 {bone1Name=dyn_grass_upper_a, bone2Name=dyn_grass_upper_base, swing1LimitDegrees=5, swing2LimitDegrees=3, twistLimitDegrees=2}
 {bone1Name=dyn_grass_upper_a_01, bone2Name=dyn_grass_upper_a, swing1LimitDegrees=5, swing2LimitDegrees=5, twistLimitDegrees=2}
 {bone1Name=dyn_grass_upper_a_02, bone2Name=dyn_grass_upper_a_01, swing1LimitDegrees=3, swing2LimitDegrees=3, twistLimitDegrees=2}
 {bone1Name=dyn_grass_upper_b, bone2Name=dyn_grass_upper_base, swing1LimitDegrees=2, swing2LimitDegrees=2, twistLimitDegrees=2}
 {bone1Name=dyn_grass_upper_b_01, bone2Name=dyn_grass_upper_b, swing1LimitDegrees=2, swing2LimitDegrees=2, twistLimitDegrees=2}
 {bone1Name=dyn_grass_upper_b_02, bone2Name=dyn_grass_upper_b_01, swing1LimitDegrees=3, swing2LimitDegrees=3, twistLimitDegrees=2}
 {bone1Name=dyn_grass_upper_c, bone2Name=dyn_grass_upper_base, swing1LimitDegrees=10, swing2LimitDegrees=10, twistLimitDegrees=2}
 {bone1Name=dyn_grass_upper_d, bone2Name=dyn_grass_upper_base, swing1LimitDegrees=10, swing2LimitDegrees=10, twistLimitDegrees=2}
 {bone1Name=dyn_grass_upper_e, bone2Name=dyn_grass_upper_base, swing1LimitDegrees=10, swing2LimitDegrees=10, twistLimitDegrees=2}
 {bone1Name=dyn_grass_upper_f, bone2Name=dyn_grass_upper_base, swing1LimitDegrees=20, swing2LimitDegrees=30, twistLimitDegrees=2}
 {bone1Name=dyn_grass_upper_g, bone2Name=dyn_grass_upper_base, swing1LimitDegrees=10, swing2LimitDegrees=10, twistLimitDegrees=2}
 {bone1Name=dyn_grass_lower_a, bone2Name=dyn_grass_lower_base, swing1LimitDegrees=30, swing2LimitDegrees=30, twistLimitDegrees=3}
 {bone1Name=dyn_grass_lower_b, bone2Name=dyn_grass_lower_base, swing1LimitDegrees=30, swing2LimitDegrees=30, twistLimitDegrees=3}
 {bone1Name=dyn_grass_lower_c, bone2Name=dyn_grass_lower_base, swing1LimitDegrees=30, swing2LimitDegrees=30, twistLimitDegrees=3}
 {bone1Name=dyn_grass_lower_d, bone2Name=dyn_grass_lower_base, swing1LimitDegrees=30, swing2LimitDegrees=30, twistLimitDegrees=3}
 {bone1Name=dyn_grass_lower_e, bone2Name=dyn_grass_lower_base, swing1LimitDegrees=30, swing2LimitDegrees=30, twistLimitDegrees=3}
 {bone1Name=dyn_grass_lower_f, bone2Name=dyn_grass_lower_base, swing1LimitDegrees=30, swing2LimitDegrees=30, twistLimitDegrees=3}
 {bone1Name=dyn_foot_a_01, bone2Name=dyn_feet_base, swing1LimitDegrees=40, swing2LimitDegrees=35, twistLimitDegrees=10}
 {bone1Name=dyn_foot_a_02, bone2Name=dyn_foot_a_01, swing1LimitDegrees=45, swing2LimitDegrees=45, twistLimitDegrees=10}
 {bone1Name=dyn_foot_b_01, bone2Name=dyn_feet_base, swing1LimitDegrees=10, swing2LimitDegrees=10, twistLimitDegrees=5}
 {bone1Name=dyn_foot_b_02, bone2Name=dyn_foot_b_01, swing1LimitDegrees=15, swing2LimitDegrees=10, twistLimitDegrees=3}
 {bone1Name=dyn_mushroom_a, bone2Name=dyn_grass_lower_base, swing1LimitDegrees=10, swing2LimitDegrees=10, twistLimitDegrees=3}
 {bone1Name=dyn_mushroom_b, bone2Name=dyn_grass_lower_base, swing1LimitDegrees=10, swing2LimitDegrees=10, twistLimitDegrees=3}
 {bone1Name=dyn_mushroom_b_01, bone2Name=dyn_mushroom_b, swing1LimitDegrees=10, swing2LimitDegrees=10, twistLimitDegrees=3}
 {bone1Name=dyn_mushroom_c, bone2Name=dyn_grass_lower_base, swing1LimitDegrees=30, swing2LimitDegrees=30, twistLimitDegrees=5}
 {bone1Name=dyn_top_handle_grass_01, bone2Name=dyn_top_handle_grass_base, swing1LimitDegrees=10, swing2LimitDegrees=10, twistLimitDegrees=2}
 {bone1Name=dyn_top_handle_grass_02, bone2Name=dyn_top_handle_grass_base, swing1LimitDegrees=10, swing2LimitDegrees=10, twistLimitDegrees=2}
 {bone1Name=dyn_top_handle_grass_03, bone2Name=dyn_top_handle_grass_base, swing1LimitDegrees=10, swing2LimitDegrees=10, twistLimitDegrees=2}
 {bone1Name=dyn_leaf, bone2Name=dyn_bottom_handle_grass_base, swing1LimitDegrees=20, swing2LimitDegrees=20, twistLimitDegrees=3}
 {
  bone1Name=dyn_bottom_handle_grass_01
  bone2Name=dyn_bottom_handle_grass_base
  swing1LimitDegrees=15
  swing2LimitDegrees=15
  twistLimitDegrees=3
 }
 {
  bone1Name=dyn_bottom_handle_grass_02
  bone2Name=dyn_bottom_handle_grass_base
  swing1LimitDegrees=15
  swing2LimitDegrees=15
  twistLimitDegrees=3
 }
 {
  bone1Name=dyn_bottom_handle_grass_03
  bone2Name=dyn_bottom_handle_grass_base
  swing1LimitDegrees=15
  swing2LimitDegrees=15
  twistLimitDegrees=3
 }
]
